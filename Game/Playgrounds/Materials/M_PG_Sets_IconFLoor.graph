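# M_PG_Sets_IconFloor is a Default Lit material in the Playgrounds content, named as an icon floor surface for sets.
# Its graph holds only three scalar Constant nodes, with no textures, parameters or math.
# No links from the constants to material outputs are shown, so which surface attributes they drive is not evident.
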Material /Game/Playgrounds/Materials/M_PG_Sets_IconFLoor.M_PG_Sets_IconFloor
graph {
 properties {pixelDepthOffsetMode=PDOM_Legacy, usedShadingModels=Default Lit}
 nodes [
  Constant() {r=0.06}
  Constant() {r=0.2}
  Constant() {r=0.56}
 ]
}
